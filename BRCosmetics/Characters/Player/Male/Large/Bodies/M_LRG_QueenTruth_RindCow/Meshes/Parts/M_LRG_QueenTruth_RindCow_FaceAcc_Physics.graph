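# Physics asset for the face accessory of the QueenTruth RindCow large male player cosmetic, simulating a dangling bandana.
# Capsule bodies cover head, neck_01 and spine_05, plus the bandana master and two dyn_bandana segments, the tip using two capsules.
# Constraints chain head and neck to the upper spine, and hang the bandana from neck_01 through an angularly free master joint.
# The two segment joints carry tight twist limits, with wider swing at the tip.
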PhysicsAsset /BRCosmetics/Characters/Player/Male/Large/Bodies/M_LRG_QueenTruth_RindCow/Meshes/Parts/M_LRG_QueenTruth_RindCow_FaceAcc_Physics
bodies {
 MSTR_bandana [{shapeType=Capsule, length=0.0984, radius=3.272, shapeName=MSTR_bandana_capsule, rotation=<rot 0 -90 0>}]
 dyn_bandana_01 [{shapeType=Capsule, length=8.7756, radius=2.3574, shapeName=dyn_bandana_01_capsule, center=<0.5679 0 -0.6119>, rotation=<rot 0 0 90>}]
 dyn_bandana_02 [
  {
   shapeType=Capsule
   length=5.5052
   radius=1.7645
   shapeName=dyn_bandana_02_capsule
   center=<0.3022 0 1.0473>
   rotation=<rot 0.0001 -3.3399 90>
  }
  {shapeType=Capsule, length=2.0242, radius=1.7645, shapeName=dyn_bandana_02_capsule, center=<3.3039 -0.7393 1.0473>, rotation=<rot 0 0 90>}
 ]
 head [{shapeType=Capsule, length=23.8612, radius=12.9704, center=<16.5742 5.1967 0.1823>, rotation=<rot -0.092 -101.6559 -88.893>}]
 neck_01 [{shapeType=Capsule, length=0, radius=6.178, center=<0.4568 4.3923 0>}]
 spine_05 [{shapeType=Capsule, length=23.5224, radius=8.8416, center=<0 6.3888 0>}]
}
constraints [
 {bone1Name=neck_01, bone2Name=spine_05, swing1LimitDegrees=45, swing2LimitDegrees=45, twistLimitDegrees=45}
 {bone1Name=head, bone2Name=neck_01, swing1LimitDegrees=45, swing2LimitDegrees=45, twistLimitDegrees=45}
 {
  bone1Name=MSTR_bandana
  bone2Name=neck_01
  swing1LimitDegrees=45
  swing1Motion=Free
  swing2LimitDegrees=45
  swing2Motion=Free
  twistLimitDegrees=45
  twistMotion=Free
 }
 {bone1Name=dyn_bandana_01, bone2Name=MSTR_bandana, swing1LimitDegrees=15, swing2LimitDegrees=20, twistLimitDegrees=5}
 {bone1Name=dyn_bandana_02, bone2Name=dyn_bandana_01, swing1LimitDegrees=15, swing2LimitDegrees=45, twistLimitDegrees=5}
]
