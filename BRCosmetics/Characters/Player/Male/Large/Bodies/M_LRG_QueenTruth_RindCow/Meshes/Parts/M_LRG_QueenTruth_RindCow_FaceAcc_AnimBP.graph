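# Post-copy physics AnimBP for the face accessory of the QueenTruth RindCow large male character part.
# The AnimGraph copies the attached parent's pose, then runs a RigidBody node simulating in spine_02 base-bone space with its own physics asset and world-gravity override.
# Each update, a nativized state function picks gravity/wind and sim-space settings from per-movement-state tables (skydive, swim, DBNO, riding), storing them in Gravity Override and RBAN Settings Out.
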AnimBlueprint /BRCosmetics/Characters/Player/Male/Large/Bodies/M_LRG_QueenTruth_RindCow/Meshes/Parts/M_LRG_QueenTruth_RindCow_FaceAcc_AnimBP
target_skeleton=&/M_LRG_QueenTruth_RindCow_FaceAcc_Skeleton
graphs [
 AnimGraph {nodes [
   n1 = RigidBody!(
    ComponentPose=Local To Component(LocalPose=Copy Pose From Mesh() {bUseAttachedParent=true})
    Alpha
    OverrideWorldGravity:Vector="0,0,0"
    ExternalForce
    SimSpaceSettings:Sim Space Settings Structure=(WorldAlpha=0,VelocityScaleZ=1,DampingAlpha=1,MaxLinearVelocity=10000,MaxAngularVelocity=10000,MaxLinearAcceleration=10000,MaxAngularAcceleration=10000,ExternalLinearDragV=(X=0,Y=0,Z=0),ExternalLinearVelocity=(X=0,Y=0,Z=0),ExternalAngularVelocity=(X=0,Y=0,Z=0))
   ) -> (Pose) {
    bClampLinearTranslationLimitToRefPose=true
    bOverrideWorldGravity=true
    overridePhysicsAsset=&/M_LRG_QueenTruth_RindCow_FaceAcc_Physics
    simulationSpace=BaseBoneSpace
    baseBoneRef {boneName=spine_02}
   }
   Output Pose(Result=Component To Local(ComponentPose=@n1)) {layerGroup=None}
  ]}
 EventGraph {nodes [
   Event BlueprintUpdateAnimation()
   > n1 = "Gravity Override State Function (Nativized)"!(
    execute
    CharacterPartAnimInstance:Custom Character Part Anim Instance Object Reference=Self-Reference()
    GravityOverrideParameterSet:Fort Gravity Override Parameter Set Structure=Get Gravity Override Params() -> (Gravity Override Params:Fort Gravity Override Parameter Set Structure)
    RigidBodyAnimNodeParameterSet:Fort Rigid Body Anim Node Parameter Set Structure=Get RBAN Params() -> (RBAN Params:Fort Rigid Body Anim Node Parameter Set Structure)
    DebugDraw:bool=false
    GameTimeInSeconds:float=GetGameTimeInSeconds()
    DebugJointName:Name=root
    StateLogicLayerName:Name=none
   ) -> (
    then
    OutGravityOverride:Vector
    OutCurrentState:Name=None
    OutGravityOverrideParameters:Fort Gravity Override Parameters Structure
    OutRigidBodyAnimNodeParameters:Fort Rigid Body Anim Node Parameters Structure
   )
   > Set Gravity Override(Gravity Override=@n1)
   > Set RBAN Settings Out(RBAN Settings Out=@n1.OutRigidBodyAnimNodeParameters)
   "Gravity Override State Function (Parameter Asset) (Nativized)"(GameTimeInSeconds=0)
  ]}
]
variables [
 Gravity Override {type=Vector}
 Gravity Override Params {
  type=FortGravityOverrideParameterSet
  default=(Skydive_GoingUp=(GravityOverride=(WindFrequency=(X=1.500000,Y=7.000000,Z=8.000000),WindAmplitude=(X=200.000000,Y=-500.000000,Z=400.000000),GravityOverride=(X=0.000000,Y=-300.000000,Z=1500.000000),JointName="root")),Skydive_FloatingDown=(GravityOverride=(WindFrequency=(X=1.500000,Y=7.000000,Z=8.000000),WindAmplitude=(X=200.000000,Y=-500.000000,Z=400.000000),WindOffset=(X=0.000000,Y=-600.000000,Z=400.000000),GravityOverride=(X=0.000000,Y=-700.000000,Z=600.000000),JointName="root")),Skydive_DivingDown=(GravityOverride=(WindFrequency=(X=1.500000,Y=12.000000,Z=10.000000),WindAmplitude=(X=50.000000,Y=300.000000,Z=100.000000),GravityOverride=(X=-1500.000000,Y=0.000000,Z=0.000000),JointName="pelvis")),Skydive_Parachuting=(GravityOverride=(WindFrequency=(X=1.500000,Y=7.000000,Z=8.000000),WindAmplitude=(X=200.000000,Y=-500.000000,Z=400.000000),GravityOverride=(X=0.000000,Y=-800.000000,Z=-980.000000),JointName="root")),OnGround_Standing=(GravityOverride=(GravityOverride=(X=0.000000,Y=0.000000,Z=-980.000000),JointName="root")),OnGround_Moving=(GravityOverride=(GravityOverride=(X=0.000000,Y=0.000000,Z=-980.000000),JointName="root")),OnGround_Moving_HighRotationRate=(GravityOverride=(GravityOverride=(X=0.000000,Y=0.000000,Z=-980.000000),JointName="root")),OnGround_Crouch_Moving=(GravityOverride=(GravityOverride=(X=0.000000,Y=0.000000,Z=-980.000000),JointName="root")),Emote_or_Melee=(GravityOverride=(GravityOverride=(X=0.000000,Y=0.000000,Z=-980.000000),JointName="root")),Swimming=(GravityOverride=(GravityOverride=(X=-800.000000,Y=0.000000,Z=0.000000),JointName="pelvis")),Front_End=(GravityOverride=(GravityOverride=(X=0.000000,Y=0.000000,Z=-980.000000),JointName="root")),DBNO=(GravityOverride=(GravityOverride=(X=-980.000000,Y=400.000000,Z=0.000000),JointName="pelvis")),DBNO_Carried=(GravityOverride=(GravityOverride=(X=-980.000000,Y=500.000000,Z=-500.000000),JointName="pelvis")),OnGround_Falling=(GravityOverride=(GravityOverride=(X=0.000000,Y=0.000000,Z=-980.000000),JointName="root")),Ziplining=(GravityOverride=(GravityOverride=(X=0.000000,Y=0.000000,Z=-980.000000),JointName="root")),Skydive_Surfing=(GravityOverride=(WindFrequency=(X=1.500000,Y=7.000000,Z=8.000000),WindAmplitude=(X=200.000000,Y=-500.000000,Z=400.000000),GravityOverride=(X=500.000000,Y=-300.000000,Z=-980.000000),JointName="root")),Riding_Standing=(GravityOverride=(GravityOverride=(X=0.000000,Y=-350.000000,Z=-980.000000),JointName="root")),Riding_Moving=(GravityOverride=(GravityOverride=(X=0.000000,Y=-450.000000,Z=-450.000000),JointName="root")),Riding_Moving_Targeting=(GravityOverride=(GravityOverride=(X=0.000000,Y=-200.000000,Z=-980.000000),JointName="root")),Tactical_Sprint=(GravityOverride=(GravityOverride=(X=0.000000,Y=-450.000000,Z=-500.000000),JointName="root")),Rail_Grinding=(GravityOverride=(GravityOverride=(X=0.000000,Y=0.000000,Z=-980.000000),JointName="root")),)
 }
 RBAN Params {
  type=FortRigidBodyAnimNodeParameterSet
  default=(Skydive_GoingUp=(RigidBodyAnimNode=(SimSpaceSettings=(WorldAlpha=1.000000,MaxLinearVelocity=500.000000,MaxAngularVelocity=3.000000,MaxLinearAcceleration=800.000000,MaxAngularAcceleration=1.000000,ExternalLinearDragV=(X=1.000000,Y=1.000000,Z=1.000000)))),Skydive_FloatingDown=(RigidBodyAnimNode=(SimSpaceSettings=(WorldAlpha=1.000000,MaxLinearVelocity=500.000000,MaxAngularVelocity=3.000000,MaxLinearAcceleration=800.000000,MaxAngularAcceleration=1.000000,ExternalLinearDragV=(X=1.000000,Y=1.000000,Z=1.000000)))),Skydive_DivingDown=(RigidBodyAnimNode=(SimSpaceSettings=(WorldAlpha=1.000000,MaxLinearVelocity=500.000000,MaxAngularVelocity=3.000000,MaxLinearAcceleration=800.000000,MaxAngularAcceleration=1.000000,ExternalLinearDragV=(X=1.000000,Y=1.000000,Z=1.000000)))),Skydive_Parachuting=(RigidBodyAnimNode=(SimSpaceSettings=(WorldAlpha=1.000000,MaxLinearVelocity=500.000000,MaxAngularVelocity=3.000000,MaxLinearAcceleration=800.000000,MaxAngularAcceleration=1.000000,ExternalLinearDragV=(X=1.000000,Y=1.000000,Z=1.000000)))),OnGround_Standing=(RigidBodyAnimNode=(SimSpaceSettings=(WorldAlpha=1.000000,MaxLinearVelocity=500.000000,MaxAngularVelocity=3.000000,MaxLinearAcceleration=800.000000,MaxAngularAcceleration=1.000000,ExternalLinearDragV=(X=1.000000,Y=1.000000,Z=1.000000)))),OnGround_Moving=(RigidBodyAnimNode=(SimSpaceSettings=(WorldAlpha=1.000000,MaxLinearVelocity=500.000000,MaxAngularVelocity=3.000000,MaxLinearAcceleration=800.000000,MaxAngularAcceleration=1.000000,ExternalLinearDragV=(X=1.000000,Y=1.000000,Z=1.000000)))),OnGround_Moving_HighRotationRate=(RigidBodyAnimNode=(SimSpaceSettings=(WorldAlpha=1.000000,MaxLinearVelocity=500.000000,MaxAngularVelocity=3.000000,MaxLinearAcceleration=800.000000,MaxAngularAcceleration=1.000000,ExternalLinearDragV=(X=1.000000,Y=1.000000,Z=1.000000)))),OnGround_Crouch_Moving=(RigidBodyAnimNode=(SimSpaceSettings=(WorldAlpha=1.000000,MaxLinearVelocity=500.000000,MaxAngularVelocity=3.000000,MaxLinearAcceleration=800.000000,MaxAngularAcceleration=1.000000,ExternalLinearDragV=(X=1.000000,Y=1.000000,Z=1.000000)))),Emote_or_Melee=(RigidBodyAnimNode=(SimSpaceSettings=(WorldAlpha=1.000000,MaxLinearVelocity=500.000000,MaxAngularVelocity=3.000000,MaxLinearAcceleration=1000.000000,MaxAngularAcceleration=1.000000,ExternalLinearDragV=(X=1.000000,Y=1.000000,Z=1.000000)))),Swimming=(RigidBodyAnimNode=(SimSpaceSettings=(WorldAlpha=1.000000,MaxLinearVelocity=500.000000,MaxAngularVelocity=3.000000,MaxLinearAcceleration=800.000000,MaxAngularAcceleration=1.000000,ExternalLinearDragV=(X=1.000000,Y=1.000000,Z=1.000000)))),Front_End=(RigidBodyAnimNode=(SimSpaceSettings=(WorldAlpha=1.000000,MaxLinearVelocity=500.000000,MaxAngularVelocity=3.000000,MaxLinearAcceleration=800.000000,MaxAngularAcceleration=1.000000,ExternalLinearDragV=(X=1.000000,Y=1.000000,Z=1.000000)))),DBNO=(RigidBodyAnimNode=(SimSpaceSettings=(WorldAlpha=1.000000,MaxLinearVelocity=500.000000,MaxAngularVelocity=3.000000,MaxLinearAcceleration=800.000000,MaxAngularAcceleration=1.000000,ExternalLinearDragV=(X=1.000000,Y=1.000000,Z=1.000000)))),DBNO_Carried=(RigidBodyAnimNode=(SimSpaceSettings=(WorldAlpha=1.000000,MaxLinearVelocity=500.000000,MaxAngularVelocity=3.000000,MaxLinearAcceleration=800.000000,MaxAngularAcceleration=1.000000,ExternalLinearDragV=(X=1.000000,Y=1.000000,Z=1.000000)))),OnGround_Falling=(RigidBodyAnimNode=(SimSpaceSettings=(WorldAlpha=1.000000,MaxLinearVelocity=500.000000,MaxAngularVelocity=3.000000,MaxLinearAcceleration=800.000000,MaxAngularAcceleration=1.000000,ExternalLinearDragV=(X=1.000000,Y=1.000000,Z=1.000000)))),Ziplining=(RigidBodyAnimNode=(SimSpaceSettings=(WorldAlpha=1.000000,MaxLinearVelocity=500.000000,MaxAngularVelocity=3.000000,MaxLinearAcceleration=800.000000,MaxAngularAcceleration=1.000000,ExternalLinearDragV=(X=1.000000,Y=1.000000,Z=1.000000)))),Skydive_Surfing=(RigidBodyAnimNode=(SimSpaceSettings=(WorldAlpha=1.000000,MaxLinearVelocity=500.000000,MaxAngularVelocity=3.000000,MaxLinearAcceleration=800.000000,MaxAngularAcceleration=1.000000,ExternalLinearDragV=(X=1.000000,Y=1.000000,Z=1.000000)))),Riding_Standing=(RigidBodyAnimNode=(SimSpaceSettings=(WorldAlpha=1.000000,MaxLinearVelocity=500.000000,MaxAngularVelocity=3.000000,MaxLinearAcceleration=800.000000,MaxAngularAcceleration=1.000000,ExternalLinearDragV=(X=1.000000,Y=1.000000,Z=1.000000)))),Riding_Moving=(RigidBodyAnimNode=(SimSpaceSettings=(WorldAlpha=1.000000,MaxLinearVelocity=500.000000,MaxAngularVelocity=3.000000,MaxLinearAcceleration=800.000000,MaxAngularAcceleration=1.000000,ExternalLinearDragV=(X=1.000000,Y=1.000000,Z=1.000000)))),Riding_Moving_Targeting=(RigidBodyAnimNode=(SimSpaceSettings=(WorldAlpha=1.000000,MaxLinearVelocity=500.000000,MaxAngularVelocity=3.000000,MaxLinearAcceleration=1000.000000,MaxAngularAcceleration=1.000000,ExternalLinearDragV=(X=1.000000,Y=1.000000,Z=1.000000)))),Tactical_Sprint=(RigidBodyAnimNode=(SimSpaceSettings=(WorldAlpha=1.000000,MaxLinearVelocity=500.000000,MaxAngularVelocity=3.000000,MaxLinearAcceleration=800.000000,MaxAngularAcceleration=1.000000,ExternalLinearDragV=(X=1.000000,Y=1.000000,Z=1.000000)))),Rail_Grinding=(RigidBodyAnimNode=(SimSpaceSettings=(WorldAlpha=1.000000,MaxLinearVelocity=500.000000,MaxAngularVelocity=3.000000,MaxLinearAcceleration=800.000000,MaxAngularAcceleration=1.000000,ExternalLinearDragV=(X=1.000000,Y=1.000000,Z=1.000000)))),)
 }
 RBAN Settings Out {type=FortRigidBodyAnimNodeParameters}
]
